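# Material function for the Aglet02 skin on the ZRF vehicle body, adding animated lettering highlights over its base color texture.
# A tiled letter mask, gated by the base color's alpha, is modulated by two counter-rotated, panning stripe/dot patterns, with a sine-driven offset available via named reroutes.
# The mask blends dark red-to-white tint into the base color, whose desaturated value derives roughness.
# Both results override BaseColor and Roughness on MF_Vehicle_Body_Base attributes.
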MaterialFunction /VehicleCosmetics/Bodies/ZRF/Skins/Aglet02/MF_Skin_ZRF_Aglet02
graph {nodes [
  n1 = TextureSample() {desc=BC, texture=&/T_Aglet02_BC}
  n2 = Multiply(A=NamedRerouteUsage(), B=Constant2Vector() {g=17, r=18.6})
  n3 = Saturate(None=Multiply(
    A=@n1.A
    B=TextureSample(UVs=@n2) {desc=Letter Mask, samplerType=SAMPLERTYPE_Masks, texture=&/T_Aglet02_LetterMask}.R
   ))
  n4 = NamedRerouteUsage()
  n5 = Constant2Vector() {g=1, r=0.5}
  n6 = Panner(Coordinate=Add(A=NamedRerouteUsage(), B=Multiply(A=@n4, B=@n5))) {constCoordinate=1, speedX=0.01, speedY=0.15}
  n7 = Constant() {r=-0.05}
  n8 = ConstantBiasScale(None=TextureSample(UVs=MaterialFunctionCall(UVs=@n6, "Rotation Angle (0-1)"=@n7) {materialFunction=&CustomRotator}.Rotated Values) {texture=/VehicleCosmetics/SharedTextures/T_Stripes_Dots_RGB}.R) {bias=-0.1, scale=2}
  n9 = NamedRerouteUsage()
  n10 = Constant2Vector() {g=1, r=0.5}
  n11 = Panner(Coordinate=Add(A=NamedRerouteUsage(), B=Multiply(A=@n9, B=@n10))) {constCoordinate=1, speedX=0.01, speedY=0.3}
  n12 = Constant() {r=0.05}
  n13 = ConstantBiasScale(None=TextureSample(UVs=MaterialFunctionCall(UVs=@n11, "Rotation Angle (0-1)"=@n12) {materialFunction=&CustomRotator}.Rotated Values) {texture=&T_Stripes_Dots_RGB}.R) {bias=-0.1, scale=2}
  n14 = Clamp(None=ConstantBiasScale(None=Add(A=@n8, B=@n13)) {bias=-0.98}) {desc=Lighten Min Values, minDefault=0.2}
  n15 = LinearInterpolate(
   A=@n1.RGB
   B=LinearInterpolate(A=Constant3Vector() {constant=<rgba 0.0816 0.0141 0.015 1>}, B=Constant3Vector() {constant=<rgba 1 1 1 1>}, Alpha=@n3)
   Alpha=Multiply(A=@n3, B=Saturate(None=@n14))
  ) {constB=0.2}
  n16 = Desaturation(None=@n15)
  n17 = Multiply(A=Clamp(None=Multiply(A=@n16) {constB=0.8}) {minDefault=0.7}) {constB=0.5, desc="Lower values = Lower Roughness"}
  FunctionOutput(None=SetMaterialAttributes!(
    MaterialAttributes=MaterialFunctionCall() {materialFunction=/VehicleCosmetics/Materials/MaterialFunctions/MF_Vehicle_Body_Base}.Result
    BaseColor=@n15
    Roughness=Saturate(None=@n17)
   ) {attributeSetTypes [#1, #2]})
  n18 = NamedRerouteUsage()
  n19 = ComponentMask(None=ConstantBiasScale(None=Sine(None=@n18) {period=30}) {bias=0.5, scale=0.25}) {r=true}
  NamedRerouteDeclaration(None=AppendVector(A=Add(A=@n19) {constB=0.5}, B=Constant() {r=1})) {name=SineOffset_Out, nodeColor=<rgba 0.5234 0 1 1>}
  NamedRerouteDeclaration(None=TextureCoordinate()) {name=TexCoord0_Out, nodeColor=<rgba 1 0 0.1172 1>}
  NamedRerouteDeclaration(None=Time()) {name=Time_Out, nodeColor=<rgba 1 0.2812 0 1>}
 ]}
